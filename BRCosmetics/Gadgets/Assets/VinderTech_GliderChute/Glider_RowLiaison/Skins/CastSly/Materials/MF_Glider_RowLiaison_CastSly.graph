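# Material function for the RowLiaison CastSly glider skin that layers an animated, reactive emissive pattern onto input material attributes.
# Two noise samples panned over pre-skinned local-position coordinates blend with a time-driven sine flow, then get masking and Fresnel shaping before tinting.
# A curve-atlas row, driven by time elapsed since a game-time parameter named as Blueprint-supplied, provides reactive progress.
# Masked lerps also darken base color and set metallic.
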MaterialFunction /BRCosmetics/Gadgets/Assets/VinderTech_GliderChute/Glider_RowLiaison/Skins/CastSly/Materials/MF_Glider_RowLiaison_CastSly
graph {nodes [
  n1 = Fresnel() {baseReflectFraction=0, exponent=2}
  Add(A=@n1) {constB=0}
  n2 = GetMaterialAttributes(None=FunctionInput() {inputName=Material, inputType=FunctionInput_MaterialAttributes}) {attributeGetTypes [#1, #2, #3]}
  n3 = NamedRerouteUsage()
  n4 = NamedRerouteUsage()
  n5 = ScalarParameter() {defaultValue=1, parameterName=PatternMetallic}
  FunctionOutput(None=SetMaterialAttributes!(
    MaterialAttributes=@n2.MaterialAttributes
    EmissiveColor=Add(A=@n2.EmissiveColor, B=@n3)
    BaseColor=LinearInterpolate(A=@n2.BaseColor, Alpha=@n4) {constB=0}
    Metallic=LinearInterpolate(A=@n2.Metallic, B=@n5, Alpha=@n4)
   ) {attributeSetTypes [#1, #2, #3]})
  n6 = TextureSampleParameter2D() {
   group=RowLiaison_CastSly
   parameterName=RLCS_Texture
   samplerType=SAMPLERTYPE_Masks
   texture=/BRCosmetics/Gadgets/Assets/VinderTech_GliderChute/Glider_RowLiaison/Skins/CastSly/Textures/T_RowLiaison_CastSly_Glider_ReactivePattern
  }
  NamedRerouteDeclaration(None=@n6.G) {name=RLCS_PatternMask, nodeColor=<rgba 1 0.3047 0 1>}
  n7 = LinearInterpolate(
   A=ScalarParameter() {defaultValue=1, group=RowLiaison_CastSly, parameterName=PatternBrightness_Stage1}
   B=ScalarParameter() {defaultValue=2, group=RowLiaison_CastSly, parameterName=PatternBrightness_Stage2}
   Alpha=NamedRerouteUsage()
  )
  n8 = ComponentMask(None=NamedRerouteUsage()) {b=true, r=true}
  n9 = Constant2Vector() {g=-0.1234, r=0.1932}
  n10 = ScalarParameter() {defaultValue=0.3, group=RowLiaison_CastSly, parameterName=PatternSpeed_Stage1}
  n11 = TextureSample(UVs=Panner(Coordinate=@n8, Speed=Multiply(A=@n9, B=@n10)) {bFractionalPart=true, speedX=0.197, speedY=-0.1235}) {
   samplerSource=SSM_Wrap_WorldGroupSettings
   samplerType=SAMPLERTYPE_Masks
   texture=/BRCosmetics/Characters/Player/Female/Medium/Bodies/F_MED_RowLiaison_Cast/Skins/Sly/Textures/T_RowLiaison_CastSly_Noise
  }
  n12 = Constant2Vector() {g=-0.0823, r=-0.3243}
  n13 = TextureSample(UVs=Panner(Coordinate=@n8, Speed=Multiply(A=@n12, B=@n10)) {bFractionalPart=true, speedX=-0.324, speedY=-0.0345}) {
   samplerSource=SSM_Wrap_WorldGroupSettings
   samplerType=SAMPLERTYPE_Masks
   texture=&T_RowLiaison_CastSly_Noise
  }
  n14 = Multiply(A=@n11.R, B=@n13.R)
  n15 = TextureSample() {
   samplerType=SAMPLERTYPE_Masks
   texture=&T_RowLiaison_CastSly_Noise
  }
  n16 = Time()
  n17 = ScalarParameter() {defaultValue=-0.5, parameterName=PatternFlowSpeed}
  n18 = Sine(None=Add(A=@n15.G, B=Multiply(A=@n16, B=@n17) {constB=-0.5}))
  n19 = Power(Base=OneMinus(None=ConstantBiasScale(None=@n18))) {constExponent=3}
  n20 = LinearInterpolate(A=Add(A=@n14, B=Multiply(A=@n19) {constB=0.05}), B=Add(A=Multiply(A=@n14) {constB=0.5}, B=@n19), Alpha=NamedRerouteUsage())
  n21 = Multiply(A=Multiply(A=NamedRerouteUsage(), B=Multiply(A=@n7, B=@n20)), B=OneMinus(None=@n1))
  n22 = Multiply(A=Multiply(A=@n1, B=NamedRerouteUsage()), B=ScalarParameter() {defaultValue=0.5, parameterName=ReactiveFresnelIntensity}) {constB=0.75}
  n23 = Multiply(
   A=LinearInterpolate(A=@n21, B=Add(A=@n21, B=@n22), Alpha=NamedRerouteUsage())
   B=ScalarParameter() {defaultValue=1, group=RowLiaison_CastSly, parameterName=RLCS_Reactivity}
  )
  NamedRerouteDeclaration(None=Multiply(
    A=VectorParameter() {group=RowLiaison_CastSly, parameterName=RLCS_ReactiveColor, defaultValue=<rgba 1 0.8665 0.5573 1>}.RGB
    B=NamedRerouteDeclaration(None=@n23) {name=FinalPattern, nodeColor=<rgba 0.8984 0 1 1>}
   )) {name=Emissive_OUT, nodeColor=<rgba 0 0.0156 1 1>}
  n24 = Subtract(A=Time(), B=ScalarParameter() {defaultValue=10000000272564224, group=RowLiaison_CastSly, parameterName=RLCS_GameTimeFromBP})
  n25 = ScalarParameter() {defaultValue=5, group=RowLiaison_CastSly, parameterName=RLCS_ReactiveTime}
  n26 = CurveAtlasRowParameter(CurveTime=Saturate(None=Divide(A=@n24, B=@n25))) {
   atlas=/BRCosmetics/Characters/Player/Female/Medium/Bodies/F_MED_RowLiaison_Cast/Skins/Sly/Textures/CA_RowLiaison_CastSly_Reactivity
   curve=/BRCosmetics/Characters/Player/Female/Medium/Bodies/F_MED_RowLiaison_Cast/Skins/Sly/Textures/C_RowLiaison_CastSly_Reactivity
   group=RowLiaison_CastSly
   parameterName=RLCS_ReactiveCurve
  }
  NamedRerouteDeclaration(None=@n26.R) {name=ReactiveProgress, nodeColor=<rgba 1 0 0.2578 1>}
  n27 = VertexInterpolator(VS=LocalPosition() {localOrigin=InstancePreSkinning}.XYZ)
  NamedRerouteDeclaration(None=Divide(A=FunctionInput(Preview=@n27.PS) {inputName=PreskinnedLocPos, sortPriority=1}) {constB=128}) {name=PreskinnedLocPos, nodeColor=<rgba 1 0.0703 0 1>}
  NamedRerouteDeclaration(None=@n6.R) {name=RLCS_DressMask, nodeColor=<rgba 1 0.7079 0.9608 1>}
  Saturate(None=ScalarParameter() {defaultValue=1, parameterName=DebugProgress})
 ]}
